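# Sound cue for the thrust sound of the Alchemy cosmetic glider.
# A Random node picks one of six equally weighted body-shake waves, and a Modulator with default settings varies its pitch and volume.
# The cue plays at reduced volume using glider thrust attenuation and concurrency, a cosmetic glider loop audio properties sheet, and the Fort_Gliders_Land sound class.
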SoundCue /BRCosmetics/Sounds/Gliders/Alchemy/Glider_Alchemy_Thrust_Cue
root=Modulator_0
nodes [
 {id=Modulator_0, children [Random_2]}
 {
  id=Random_2
  children [WavePlayer_9, WavePlayer_10, WavePlayer_11, WavePlayer_12, WavePlayer_13, WavePlayer_14]
  properties {weights [1,1,1,1,1,1]}
 }
 {id=WavePlayer_10, properties {soundWaveAssetPtr=&/Glider_Alchemy_Thrust_Body_Shake_02}}
 {id=WavePlayer_11, properties {soundWaveAssetPtr=&/Glider_Alchemy_Thrust_Body_Shake_03}}
 {id=WavePlayer_12, properties {soundWaveAssetPtr=&/Glider_Alchemy_Thrust_Body_Shake_04}}
 {id=WavePlayer_13, properties {soundWaveAssetPtr=&/Glider_Alchemy_Thrust_Body_Shake_05}}
 {id=WavePlayer_14, properties {soundWaveAssetPtr=&/Glider_Alchemy_Thrust_Body_Shake_06}}
 {id=WavePlayer_9, properties {soundWaveAssetPtr=&/Glider_Alchemy_Thrust_Body_Shake_01}}
]
properties {
 attenuationSettings=&Fort_Player_Glider_Thrust_Athena_Att
 audioPropertiesSheet=&APS_Cosmetic_Glider_Loop
 maxDistance=12500
 soundClassObject=&Fort_Gliders_Land
 volumeMultiplier=0.7
 concurrencySet [&Athena_Glider_Thrust_Conc]
}
