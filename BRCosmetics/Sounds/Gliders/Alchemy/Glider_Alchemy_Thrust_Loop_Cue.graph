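# Looping thrust sound cue for the Alchemy glider, spatialized with the player glider thrust attenuation.
# Two submixes, a thrust loop and a thrust wind loop, combine in a root mixer that boosts the thrust layer.
# Each submix plays three copies of its loop through continuous modulators driven by Forwards, Sideways and Backwards parameters, which scale volume and pitch, with the wind layer in higher output ranges.
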SoundCue /BRCosmetics/Sounds/Gliders/Alchemy/Glider_Alchemy_Thrust_Loop_Cue
root=Mixer_2
nodes [
 {id=Mixer_0, children [ModulatorContinuous_0, ModulatorContinuous_1, ModulatorContinuous_2], properties {inputVolume [1,1,1]}}
 {id=Mixer_1, children [ModulatorContinuous_3, ModulatorContinuous_4, ModulatorContinuous_5], properties {inputVolume [1,1,1]}}
 {id=Mixer_2, children [Mixer_0,Mixer_1], properties {inputVolume [1.5,1]}}
 {
  id=ModulatorContinuous_0
  children [WavePlayer_0]
  properties {
   pitchModulationParams {default=0, minOutput=0.8, parameterName=Forwards}
   volumeModulationParams {default=0, maxOutput=0.75, minOutput=0.1, parameterName=Forwards}
  }
 }
 {
  id=ModulatorContinuous_1
  children [WavePlayer_1]
  properties {
   pitchModulationParams {default=0, minOutput=0.8, parameterName=Sideways}
   volumeModulationParams {default=0, maxOutput=0.75, minOutput=0.1, parameterName=Sideways}
  }
 }
 {
  id=ModulatorContinuous_2
  children [WavePlayer_2]
  properties {
   pitchModulationParams {default=0, minOutput=0.8, parameterName=Backwards}
   volumeModulationParams {default=0, maxOutput=0.75, minOutput=0.1, parameterName=Backwards}
  }
 }
 {
  id=ModulatorContinuous_3
  children [WavePlayer_3]
  properties {
   pitchModulationParams {default=0, maxOutput=1.2, minOutput=1, parameterName=Forwards}
   volumeModulationParams {default=0, maxOutput=1.5, minOutput=0.35, parameterName=Forwards}
  }
 }
 {
  id=ModulatorContinuous_4
  children [WavePlayer_4]
  properties {
   pitchModulationParams {default=0, maxOutput=1.2, minOutput=1, parameterName=Sideways}
   volumeModulationParams {default=0, maxOutput=1.5, minOutput=0.35, parameterName=Sideways}
  }
 }
 {
  id=ModulatorContinuous_5
  children [WavePlayer_5]
  properties {
   pitchModulationParams {default=0, maxOutput=1.2, minOutput=1, parameterName=Backwards}
   volumeModulationParams {default=0, maxOutput=1.5, minOutput=0.35, parameterName=Backwards}
  }
 }
 {id=WavePlayer_0, properties {bLooping=true, soundWaveAssetPtr=&/Glider_Alchemy_Thrust_Loop}}
 {id=WavePlayer_1, properties {bLooping=true, soundWaveAssetPtr=&Glider_Alchemy_Thrust_Loop}}
 {id=WavePlayer_2, properties {bLooping=true, soundWaveAssetPtr=&Glider_Alchemy_Thrust_Loop}}
 {id=WavePlayer_3, properties {bLooping=true, soundWaveAssetPtr=&/Glider_Alchemy_Thrust_Wind_Loop}}
 {id=WavePlayer_4, properties {bLooping=true, soundWaveAssetPtr=&Glider_Alchemy_Thrust_Wind_Loop}}
 {id=WavePlayer_5, properties {bLooping=true, soundWaveAssetPtr=&Glider_Alchemy_Thrust_Wind_Loop}}
]
properties {
 attenuationSettings=&Fort_Player_Glider_Thrust_Athena_Att
 audioPropertiesSheet=&APS_Cosmetic_Glider_Loop
 maxDistance=12500
 pitchMultiplier=0.99
 soundClassObject=&Fort_Gliders_Land
 volumeMultiplier=0.85
}
