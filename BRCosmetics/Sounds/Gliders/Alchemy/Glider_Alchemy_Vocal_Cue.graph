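# Sound cue for the Alchemy glider that plays occasional thrust body vocal one-shots.
# A random delay of up to 1.5 seconds feeds a modulator, then an equally weighted random pick among six vocal waves and four empty slots, so some triggers stay silent.
# It uses glider thrust attenuation, and overridden concurrency allows one instance per owner and blocks new ones while it plays.
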SoundCue /BRCosmetics/Sounds/Gliders/Alchemy/Glider_Alchemy_Vocal_Cue
root=Delay_0
nodes [
 {id=Delay_0, children [Modulator_1], properties {delayMax=1.5}}
 {id=Modulator_1, children [Random_0]}
 {
  id=Random_0
  children [WavePlayer_0,WavePlayer_4,WavePlayer_5,WavePlayer_6,WavePlayer_7,WavePlayer_8,null,null,null,null]
  properties {weights [1,1,1,1,1,1,1,1,1,1]}
 }
 {id=WavePlayer_0, properties {soundWaveAssetPtr=&/Glider_Alchemy_Thrust_Body_Vocal_06}}
 {id=WavePlayer_4, properties {soundWaveAssetPtr=&/Glider_Alchemy_Thrust_Body_Vocal_01}}
 {id=WavePlayer_5, properties {soundWaveAssetPtr=&/Glider_Alchemy_Thrust_Body_Vocal_02}}
 {id=WavePlayer_6, properties {soundWaveAssetPtr=&/Glider_Alchemy_Thrust_Body_Vocal_03}}
 {id=WavePlayer_7, properties {soundWaveAssetPtr=&/Glider_Alchemy_Thrust_Body_Vocal_04}}
 {id=WavePlayer_8, properties {soundWaveAssetPtr=&/Glider_Alchemy_Thrust_Body_Vocal_05}}
]
properties {
 attenuationSettings=&Fort_Player_Glider_Thrust_Athena_Att
 bOverrideConcurrency=true
 maxDistance=12500
 soundClassObject=&Fort_Gliders_Land
 volumeMultiplier=0.9
 concurrencyOverrides {bLimitToOwner=true, maxCount=1, resolutionRule=PreventNew}
 concurrencySet [&Athena_Glider_Thrust_Conc]
}
