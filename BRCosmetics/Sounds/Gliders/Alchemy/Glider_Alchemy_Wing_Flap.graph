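# Sound cue for the Alchemy glider's wing flap thrust sound, a glider cosmetic.
# A Random node picks one of five equally weighted wing flap waves, feeding a Modulator that varies volume on each play.
# The cue plays at reduced volume with player glider thrust attenuation, a glider sound class and a thrust concurrency set.
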SoundCue /BRCosmetics/Sounds/Gliders/Alchemy/Glider_Alchemy_Wing_Flap
root=Modulator_2
nodes [
 {id=Modulator_2, children [Random_1], properties {volumeMin=0.75}}
 {id=Random_1, children [WavePlayer_1, WavePlayer_2, WavePlayer_3, WavePlayer_15, WavePlayer_16], properties {weights [1,1,1,1,1]}}
 {id=WavePlayer_1, properties {soundWaveAssetPtr=&/Glider_Alchemy_Thrust_Wing_Flap_01}}
 {id=WavePlayer_15, properties {soundWaveAssetPtr=&/Glider_Alchemy_Thrust_Wing_Flap_04}}
 {id=WavePlayer_16, properties {soundWaveAssetPtr=&/Glider_Alchemy_Thrust_Wing_Flap_05}}
 {id=WavePlayer_2, properties {soundWaveAssetPtr=&/Glider_Alchemy_Thrust_Wing_Flap_02}}
 {id=WavePlayer_3, properties {soundWaveAssetPtr=&/Glider_Alchemy_Thrust_Wing_Flap_03}}
]
properties {
 attenuationSettings=&Fort_Player_Glider_Thrust_Athena_Att
 maxDistance=12500
 soundClassObject=&Fort_Gliders_Land
 volumeMultiplier=0.25
 concurrencySet [&Athena_Glider_Thrust_Conc]
}
